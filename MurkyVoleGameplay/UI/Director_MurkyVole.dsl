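; Dynamic UI director for MurkyVole that adds or removes the Hover Jets UI scene for the local player.
; It rebinds gameplay tag listeners on possessed pawn changes, tracking inventory, flying and vehicle flags.
; The scene shows only with Hover Jets in inventory and neither flying nor in a vehicle, with a delta gate filtering redundant updates.
; Flying is set by the RocketRam WindingUp tag but cleared on MovementMode.Flying removal.
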(event EventBeginPlay
  ; Begin/EndPlay
  (bind _output_get (Variables|Default|SetBoundPlayerController (DynamicUIDirector|GetOwningLocalPlayerController)))
  (Pawn|BindEventtoOnPossessedPawnChanged _output_get (EventDispatchers|CreateEvent :!event OnPossessedPawnChanged))
  (CallFunction|ListenforPawnEvents (Pawn|GetControlledPawn (Variables|Default|GetBoundPlayerController))))

(event EventEndPlay
  (params (EndPlayReason :type EEndPlayReason%20Enum :default "Destroyed"))
  (bind _boundplayercontroller (Variables|Default|GetBoundPlayerController))
  ; Begin/EndPlay
  (Utilities|IsValid _boundplayercontroller
    (:Is%20Valid
      (Pawn|UnbindEventfromOnPossessedPawnChanged _boundplayercontroller (EventDispatchers|CreateEvent :!event OnPossessedPawnChanged)))))

(event Custom|ListenForPawnEvents
  (params (Pawn :type Pawn%20Object%20Reference))
  ;[0 0 0] Listen For Pawn Events
  (bind _asfort_player_pawn (Utilities|Casting|CastToFortPlayerPawn Pawn))
  (Variables|Default|SetPlayerPawn _asfort_player_pawn)
  (CallFunction|ListenforHoverJetsinInventory)
  (CallFunction|ListenforFlyingState)
  (CallFunction|ListenforVehicleState))

(event Custom|ListenForHoverJetsInInventory
  (bind _hoverjetsininventorytag (Variables|HoverJets|GetHoverJetsInInventoryTag))
  ; Listen for Hover Jets in Inventory
  (Ability|Tasks|WaitGameplayTagAddToActor (Variables|Default|GetPlayerPawn) _hoverjetsininventorytag
    (:then
      (Ability|Tasks|WaitGameplayTagRemoveFromActor (Variables|Default|GetPlayerPawn) _hoverjetsininventorytag
        (:Removed
          (Variables|HoverJets|SetIsHoverJetsInInventory false)
          (Default|UpdateHoverJetSceneState))))
    (:Added
      (Variables|HoverJets|SetIsHoverJetsInInventory true)
      (Default|UpdateHoverJetSceneState))))

(event Custom|ListenForFlyingState
  ; Listen for Flying State
  (Ability|Tasks|WaitGameplayTagAddToActor (Variables|Default|GetPlayerPawn) "(TagName=\"Granted.Athena.RocketRam.State.WindingUp\")"
    (:then
      (Ability|Tasks|WaitGameplayTagRemoveFromActor (Variables|Default|GetPlayerPawn) "(TagName=\"MovementMode.Flying\")"
        (:Removed
          (Variables|Default|SetIsFlying? false)
          (Default|UpdateHoverJetSceneState))))
    (:Added
      (Variables|Default|SetIsFlying? true)
      (Default|UpdateHoverJetSceneState))))

(event Custom|ListenForVehicleState
  ; ListenForVehicleState
  (Ability|Tasks|WaitGameplayTagAddToActor (Variables|Default|GetPlayerPawn) "(TagName=\"Gameplay.Status.InVehicle\")"
    (:then
      (Ability|Tasks|WaitGameplayTagRemoveFromActor (Variables|Default|GetPlayerPawn) "(TagName=\"Gameplay.Status.InVehicle\")"
        (:Removed
          (Variables|Default|SetIsInVehicle? false)
          (Default|UpdateHoverJetSceneState))))
    (:Added
      (Variables|Default|SetIsInVehicle? true)
      (Default|UpdateHoverJetSceneState))))

(fn OnPossessedPawnChanged
  (params (OldPawn :type Pawn%20Object%20Reference) (NewPawn :type Pawn%20Object%20Reference))
  (CallFunction|ListenforPawnEvents NewPawn))

(fn SetHoverJetSceneActive
  (params (IsActive :type Boolean))
  (bind _scene_hoverjets (Variables|HoverJets|GetScene_HoverJets))
  (bind _updated_variable (Utilities|FlowControl|DeltaGateWildcard (Variables|HoverJets|GetIsHoverJetsSceneActive) IsActive)
    (:Success
      (if _updated_variable
        (DynamicUIDirector|AddScene _scene_hoverjets)
        (else
          (DynamicUIDirector|RemoveScene _scene_hoverjets))))))

(fn UpdateHoverJetSceneState
  (Default|SetHoverJetSceneActive (and (Variables|HoverJets|GetIsHoverJetsInInventory) (not (or (Variables|Default|GetIsFlying?) (Variables|Default|GetIsInVehicle?))))))

(fn ConstructionScript)
